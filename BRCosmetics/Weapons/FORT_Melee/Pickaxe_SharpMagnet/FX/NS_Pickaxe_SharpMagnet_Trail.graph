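# Niagara system for the SharpMagnet pickaxe swing trail, drawn by a single ribbon emitter over a short loop.
# AnimTrail modules spawn and update short-lived particles from skeletal mesh bones.
# Spawn color follows time of day, with alpha fade curves over emitter age blended by User.SwingSectionIndex.
# Loop age narrows ribbon width and ramps one dynamic material parameter, another takes a per-emitter random value, and both ScaleColor modules are disabled.
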NiagaraSystem /BRCosmetics/Weapons/FORT_Melee/Pickaxe_SharpMagnet/FX/NS_Pickaxe_SharpMagnet_Trail
dependencies {
 usedDataInterfaces [&NiagaraDataInterfaceCurve]
 usedDynamicInputs []
 usedModules [
  &self:SystemSpawnScript.SystemScriptSource.SystemScriptGraph.NiagaraNodeAssignment_0.SetVariables_#1
  &self:SystemSpawnScript.SystemScriptSource.SystemScriptGraph.NiagaraNodeAssignment_2.SetVariables_#2
  &SystemState
 ]
}
emitters {Round_Trail {
  emitterUpdateScript [
   &EmitterState(Life Cycle Mode=::NewEnumerator1, Loop Behavior=::NewEnumerator1, Loop Duration=0.5, UseLoopDelay=false)
   &FN_SpawnAnimTrail()
   /SetVariables(Emitter.Random:NiagaraFloat=&RandomRangeFloat(Minimum=0, Maximum=0.1)) {isSetParameters=true}
  ]
  particleSpawnScript [
   &FN_InitializeAnimTrailBoneIndecies()
   &FN_InitializeAnimTrail(Lifetime=0.15, SkeletalMesh=$System.Skeletal Mesh)
   &DynamicMaterialParameters(
    Index 0 Param 4=$Emitter.Random
    Module.Param0WriteEnabled=false
    Module.Param1WriteEnabled=false
    Module.Param2WriteEnabled=false
    Module.Param3WriteEnabled=true
   ) {module=DynamicMaterialParameters001}
   &Color(Color=&MultiplyLinearColor(
     Linear Color=&FNTimeOfDayColor(
      Night Color=<rgba 0.05 0.05 0.0333 1>
      Morning Color=<rgba 0.2 0.2 0.1333 1>
      Day Color=<rgba 0.2 0.2 0.1333 1>
      Evening Color=<rgba 0.1 0.1 0.0667 1>
     )
     Scale Factor=&MakeFloat4FromVectorAndFloat(
      "Vector (RGB)"=&VectorFromFloat(Value=0.8)
      "Float (Alpha)"=/Niagara/DynamicInputs/Math/LerpMultipleFloats(
       Float Count:ENiagara_MultipleLerpCount=::NewEnumerator1
       1:NiagaraFloat=&FloatFromCurve(FloatCurve=DataInterface() {propertyValues {Curve {keys [<at 0.2 1>,<at 0.28 0>]}}}, CurveIndex=$Emitter.Age)
       2:NiagaraFloat=&FloatFromCurve(FloatCurve=DataInterface() {propertyValues {Curve {keys [<at 0.1 1>,<at 0.2 0>]}}}, CurveIndex=$Emitter.Age)
       3:NiagaraFloat=&FloatFromCurve(FloatCurve=DataInterface() {propertyValues {Curve {keys [<at 0.1 1>,<at 0.2 0>]}}}, CurveIndex=$Emitter.Age)
       4:NiagaraFloat=&FloatFromCurve(FloatCurve=DataInterface() {propertyValues {Curve {keys [<at 0.1 1>,<at 0.15 0>]}}}, CurveIndex=$Emitter.Age)
       5:NiagaraFloat=?
       Alpha:NiagaraFloat=&Divide_Float(A=&MakeFloatFromInt(Int=$User.SwingSectionIndex), B=3)
      )
     )
    ))
  ]
  particleUpdateScript [
   &FN_UpdateAnimTrail(SkeletalMesh=$User.SkeletalMesh) {module=FN_UpdateAnimTrail001}
   &SolveForcesAndVelocity(Clamp Velocity=false, Limit Acceleration=false)
   &ScaleColor(
    Scale Alpha=&Multiply_Float(
     A=&FloatFromCurve(FloatCurve=DataInterface() {propertyValues {Curve {keys [
          {arriveTangent=-1.8862, interpMode=RCIM_Cubic, leaveTangent=-1.8862, tangentMode=RCTM_User, time=0.2, value=1.25}
          {interpMode=RCIM_Cubic, tangentMode=RCTM_User, time=1, value=0}
         ]}}})
     B=&Lerp_Float(
      A=1
      B=&FloatFromCurve(FloatCurve=DataInterface() {propertyValues {Curve {keys [
           {arriveTangent=0.6814, interpMode=RCIM_Cubic, leaveTangent=0.6814, tangentMode=RCTM_User, time=0, value=0.1}
           {arriveTangent=2.5, interpMode=RCIM_Cubic, leaveTangent=2.5, time=0.15, value=0.3}
           {arriveTangent=1.5958, time=0.2, value=0.6}
           {arriveTangent=0.0224, interpMode=RCIM_Cubic, leaveTangent=0.0224, tangentMode=RCTM_User, time=1, value=1}
          ]}}})
      Alpha=&FloatFromCurve(
       FloatCurve=DataInterface() {propertyValues {Curve {keys [<at 0 0 RCIM_Constant>,<at 0.4 1 RCIM_Constant>]}}}
       CurveIndex=$Emitter.NormalizedLoopAge
      )
     )
    )
    Scale RGB=&Multiply_VectorByFloat(Vector=<1 1 1>, Float=&FNTimeOfDayScalar(Night Scale=0.33, Morning Scale=1, Evening Scale=1))
    ScaleA=true
    ScaleRGB=true
   ) {enabled=false, module=ScaleColor001}
   &ScaleRibbonWidth(Ribbon Width Scale=&FloatFromCurve(
     FloatCurve=DataInterface() {propertyValues {Curve {keys [{arriveTangent=-1.7104, interpMode=RCIM_Cubic, leaveTangent=-1.7104, tangentMode=RCTM_User, time=0.4, value=1}, <at 1 0.5 RCIM_Cubic>]}}}
     CurveIndex=$Emitter.NormalizedLoopAge
    ))
   &DynamicMaterialParameters(
    Index 0 Param 3=&FloatFromCurve(
     FloatCurve=DataInterface() {propertyValues {Curve {keys [<at 0.25 0 RCIM_Cubic>,<at 0.4 1 RCIM_Cubic>]}}}
     CurveIndex=$Emitter.NormalizedLoopAge
    )
    Module.Param0WriteEnabled=false
    Module.Param1WriteEnabled=false
    Module.Param2WriteEnabled=true
    Module.Param3WriteEnabled=false
   )
   &ScaleColor(
    Scale Alpha=&LerpMultipleFloats(
     Float Count:ENiagara_MultipleLerpCount=::NewEnumerator1
     1:NiagaraFloat=1
     2:NiagaraFloat=1
     3:NiagaraFloat=1
     4:NiagaraFloat=&FloatFromCurve(FloatCurve=DataInterface() {propertyValues {Curve {keys [<at 0.2 1>,<at 0.25 0>]}}}, CurveIndex=$Emitter.Age)
     5:NiagaraFloat=?
     Alpha:NiagaraFloat=&Divide_Float(A=&MakeFloatFromInt(Int=$User.SwingSectionIndex), B=3)
    )
    ScaleA=true
    ScaleRGB=true
   ) {enabled=false}
  ]
  rendererClasses [&NiagaraRibbonRendererProperties]
  renderers [{rendererClass=&NiagaraRibbonRendererProperties}]
 }}
summary {
 systemName=NS_Pickaxe_SharpMagnet_Trail
 emitters [{emitterName=Round_Trail, rendererClasses [&NiagaraRibbonRendererProperties]}]
 userVariables [
  User.bisSwinging {type=NiagaraBool}
  User.RibbonIDOffset {type=NiagaraInt32}
  User.SkeletalMesh {
   type=NiagaraDataInterfaceSkeletalMesh
   defaultValue {
    struct=NiagaraExt_VariableValue_DataInterface
    value {dataInterface=&self:NiagaraDataInterfaceSkeletalMesh_1}
   }
  }
  User.SwingSectionIndex {type=NiagaraInt32}
 ]
}
